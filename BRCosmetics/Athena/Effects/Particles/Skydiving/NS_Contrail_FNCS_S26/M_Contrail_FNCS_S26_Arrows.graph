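# Unlit Niagara mesh-particle material for the arrows in the FNCS S26 skydiving contrail.
# Vertex offset rotates the mesh around pivots stored in UV1/UV2, adds outward wobble from two scrolling noise samples, and offsets against a dynamic-parameter velocity.
# Vertex color channels scale wobble, velocity offset and spin.
# Color sums a view-normal matcap with a scrolling color texture sampled by the same UVs, then tints it by time of day.
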Material /BRCosmetics/Athena/Effects/Particles/Skydiving/NS_Contrail_FNCS_S26/M_Contrail_FNCS_S26_Arrows
graph {
 properties {bUsedWithNiagaraMeshParticles=true, pixelDepthOffsetMode=PDOM_Legacy, shadingModel=MSM_Unlit, usedShadingModels=Unlit}
 nodes [
  n1 = Constant3Vector() {constant=<rgba 0 0 1 0>}
  n2 = VertexColor()
  n3 = DynamicParameter() {defaultValue=<rgba 1 50 0 1>, paramNames [RotationSpeed, WobbleIntensity, ArrowSize, Param4]}
  n4 = Multiply(A=@n2.B, B=Multiply(A=Time(), B=@n3.RotationSpeed) {constB=0.5})
  n5 = AppendVector(A=TextureCoordinate() {coordinateIndex=1}, B=ComponentMask(None=TextureCoordinate() {coordinateIndex=2}) {g=true})
  n6 = TransformPosition(None=WorldPosition().XYZ) {transformSourceType=TRANSFORMPOSSOURCE_World, transformType=TRANSFORMPOSSOURCE_Instance}
  n7 = Normalize(VectorInput=AppendVector(A=ComponentMask(None=@n5) {g=true, r=true}, B=Constant()))
  n8 = Divide(A=ComponentMask(None=@n5) {b=true}) {constB=300}
  n9 = Time()
  n10 = TextureSample(UVs=Add(A=@n8, B=Multiply(A=@n9) {constB=0.4})) {samplerType=SAMPLERTYPE_LinearColor, texture=&/T_Contrail_FNCS_S26_Noises}
  n11 = Time()
  n12 = TextureSample(UVs=Add(A=@n8, B=Multiply(A=@n11) {constB=0.2})) {samplerType=SAMPLERTYPE_LinearColor, texture=&T_Contrail_FNCS_S26_Noises}
  n13 = Multiply(A=Max(A=@n10.R, B=@n12.G)) {constB=2}
  n14 = Multiply(A=@n2.R, B=Multiply(A=@n3.WobbleIntensity, B=Subtract(A=@n13)) {constB=50})
  n15 = Multiply(A=DynamicParameter() {parameterIndex=1, defaultValue=<rgba 0 0 0 1>, paramNames [VelX,VelY,VelZ,Param4]}.RGB) {constB=-1}
  Add(
   A=Add(A=RotateAboutAxis(NormalizedRotationAxis=@n1, RotationAngle=@n4, PivotPoint=@n5, Position=@n6), B=Multiply(A=@n7, B=@n14))
   B=Multiply(A=VertexColor().G, B=Multiply(A=@n15) {constB=0.015})
  )
  n16 = Transform(None=VertexNormalWS()) {transformSourceType=TRANSFORMSOURCE_World, transformType=TRANSFORM_View}
  n17 = Add(A=Multiply(A=ComponentMask(None=@n16) {g=true, r=true}) {constB=0.5}) {constB=0.5}
  n18 = TextureSample(UVs=@n17) {
   samplerType=SAMPLERTYPE_LinearGrayscale
   texture=&/T_Contrail_FNCS_S26_Matcap
  }
  n19 = Multiply(A=Add(A=Time(), B=Multiply(A=@n2.B) {constB=3})) {constB=0.2}
  n20 = Add(
   A=@n18.R
   B=TextureSample(UVs=Add(A=@n17, B=@n19)) {texture=&/T_Contrail_FNCS_S26_Color2}.RGB
  )
  MaterialFunctionCall(
   Color_Morning=@n20
   Color_Day=@n20
   Color_Evening=Multiply(A=Constant3Vector() {constant=<rgba 0.71 0.52 0.44 1>}, B=@n20)
   Color_Night=Multiply(A=Constant3Vector() {constant=<rgba 0.29 0.47 0.71 1>}, B=@n20)
  ) {materialFunction=&Color_Time_of_Day}
  Subtract()
  Transform(None=@n5) {transformSourceType=TRANSFORMSOURCE_Local}
 ]
}
