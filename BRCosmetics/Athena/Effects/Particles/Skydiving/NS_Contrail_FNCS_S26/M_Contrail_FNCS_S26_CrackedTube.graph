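# Lit Niagara mesh-particle material for the cracked tube in the FNCS S26 skydiving contrail.
# UV-encoded pivots and dynamic parameters drive switchable world position offset: noise-modulated radial piece expansion, scaling, vertical offsets and vertex-green-weighted trailing against velocity.
# Two scrolling noise samples, combined by minimum, shape both that expansion and the tinted emissive.
# Base color follows time of day, with an optional vertex-red blend to an inside color.
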Material /BRCosmetics/Athena/Effects/Particles/Skydiving/NS_Contrail_FNCS_S26/M_Contrail_FNCS_S26_CrackedTube
graph {
 properties {
  bCastRayTracedShadows=false
  bFullyRough=true
  bUsedWithNiagaraMeshParticles=true
  pixelDepthOffsetMode=PDOM_Legacy
  usedShadingModels=Default Lit
 }
 nodes [
  n1 = TextureCoordinate() {coordinateIndex=1}
  n2 = OneMinus(None=ComponentMask(None=TextureCoordinate() {coordinateIndex=2}) {g=true})
  NamedRerouteDeclaration(None=Transform(None=AppendVector(A=@n1, B=@n2)) {transformSourceType=TRANSFORMSOURCE_Local}) {name=PiecesPivotPoint, nodeColor=<rgba 1 0.7619 0 1>}
  n3 = VertexColor()
  n4 = Multiply(A=DynamicParameter() {parameterIndex=1, defaultValue=<rgba 0 0 0 1>, paramNames [VelX,VelY,VelZ,Param4]}.RGB) {constB=-1}
  NamedRerouteDeclaration(None=Multiply(A=Power(Base=@n3.G) {constExponent=2.4}, B=Multiply(A=@n4) {constB=0.022})) {name=MovementInteraction, nodeColor=<rgba 1 0.3047 0 1>}
  n5 = NamedRerouteUsage()
  n6 = Multiply(A=TransformPosition(None=WorldPosition().XYZ) {transformSourceType=TRANSFORMPOSSOURCE_World, transformType=TRANSFORMPOSSOURCE_Instance}) {constB=-1}
  n7 = DynamicParameter() {defaultValue=<rgba 5 1 1 0>, paramNames [PieceExpandAmmount, PieceScaleAmmount, MotionMask, ZMask]}
  NamedRerouteDeclaration(None=Multiply(A=Add(A=@n5, B=@n6), B=OneMinus(None=@n7.PieceScaleAmmount))) {name=ScaleAmmount, nodeColor=<rgba 0.875 0 1 1>}
  n8 = VertexColor()
  NamedRerouteDeclaration(None=Multiply(A=Constant3Vector() {constant=<rgba 0 0 85 0>}, B=Multiply(A=@n8.G, B=@n7.ZMask))) {name=ZOffset, nodeColor=<rgba 0.0771 0.1146 0.1089 1>}
  n9 = Constant3Vector() {constant=<rgba 0.8281 0.6037 0.5116 1>}
  n10 = VectorParameter() {parameterName=BaseColor, defaultValue=<rgba 0.8594 0.8581 0.8549 1>}
  n11 = Multiply(
   A=MaterialFunctionCall(
    Color_Morning=Constant3Vector() {constant=<rgba 0.8281 0.772 0.6267 1>}
    Color_Day=@n9
    Color_Evening=@n9
    Color_Night=Constant3Vector() {constant=<rgba 0.5263 0.6724 0.8281 1>}
   ) {materialFunction=&Color_Time_of_Day}.Result
   B=@n10.RGB
  )
  n12 = LinearInterpolate(A=@n10.RGB, B=VectorParameter() {parameterName=InsideColor, defaultValue=<rgba 0 0 0 1>}.RGB, Alpha=VertexColor().R)
  NamedRerouteDeclaration(None=StaticSwitchParameter(True=Multiply(A=@n11, B=@n12), False=@n11) {parameterName=UseVertexColor}) {name=BaseColor, nodeColor=<rgba 0.0673 1 0.6851 1>}
  n13 = ScalarParameter() {parameterName=EmissiveIntensity}
  n14 = NamedRerouteUsage()
  n15 = Divide(A=ComponentMask(None=@n14) {b=true}) {constB=300}
  n16 = Time()
  n17 = Constant() {r=-1}
  n18 = TextureSample(UVs=Add(A=@n15, B=Multiply(A=@n16, B=@n17) {constB=0.4})) {samplerType=SAMPLERTYPE_LinearColor, texture=&/T_Contrail_FNCS_S26_Noises}
  n19 = Time()
  n20 = Constant() {r=-0.55}
  n21 = TextureSample(UVs=Add(A=@n15, B=Multiply(A=@n19, B=@n20) {constB=0.2})) {samplerType=SAMPLERTYPE_LinearColor, texture=&T_Contrail_FNCS_S26_Noises}
  n22 = Power(Base=Min(A=@n18.R, B=@n21.G)) {constExponent=3}
  n23 = Saturate(None=Multiply(A=@n22) {constB=6})
  NamedRerouteDeclaration(None=Multiply(A=VectorParameter() {parameterName=EmissionColor, defaultValue=<rgba 0.2568 1.2816 2 1>}.RGB, B=Multiply(A=@n13, B=@n23))) {name=Emission, nodeColor=<rgba 0.9164 1 0.916 1>}
  NamedRerouteUsage()
  NamedRerouteUsage()
  Power(Base=Saturate(None=Subtract(A=@n8.G) {constB=0.3})) {constExponent=3}
  n24 = Normalize(VectorInput=AppendVector(A=ComponentMask(None=@n14) {g=true, r=true}, B=Constant()))
  n25 = Add(A=@n14, B=Multiply(A=Time(), B=Constant() {r=0.4}) {constB=0.4})
  n26 = Add(A=Multiply(A=ComponentMask(None=@n25) {r=true}) {constB=1.255}, B=Multiply(A=ComponentMask(None=@n25) {g=true}) {constB=2.55})
  n27 = Multiply(A=NamedRerouteDeclaration(None=Frac(None=@n26)) {name=PieceRandomValue, nodeColor=<rgba 0.8968 0 1 1>}, B=@n22)
  n28 = DynamicParameter() {defaultValue=<rgba 5 1 1 0>, paramNames [PieceExpandAmmount, PieceScaleAmmount, MotionMask, ZMask]}
  n29 = LinearInterpolate(A=Multiply(A=Saturate(None=@n27), B=@n28.PieceExpandAmmount) {constB=41}, B=@n28.PieceExpandAmmount, Alpha=@n28.MotionMask)
  n30 = Add(A=Add(A=Multiply(A=@n24, B=@n29) {constB=40.9035}, B=NamedRerouteUsage()), B=NamedRerouteUsage())
  n31 = NamedRerouteUsage()
  n32 = Subtract(A=VertexColor().G) {constB=0.75}
  n33 = TextureCoordinate() {coordinateIndex=2}
  n34 = AppendVector(A=TextureCoordinate() {coordinateIndex=1}, B=OneMinus(None=ComponentMask(None=@n33) {g=true}))
  n35 = Add(A=Multiply(A=ComponentMask(None=@n34) {r=true}) {constB=1.528}, B=Multiply(A=ComponentMask(None=@n34) {g=true}) {constB=3.167})
  n36 = Multiply(A=Saturate(None=Multiply(A=@n32) {constB=4}), B=Power(Base=Frac(None=@n35)) {constExponent=1.3})
  n37 = Transform(None=Constant3Vector() {constant=<rgba 0 0 12 0>}) {transformSourceType=TRANSFORMSOURCE_Local}
  StaticSwitchParameter(True=Add(A=Add(A=@n30, B=@n31), B=Multiply(A=@n36, B=@n37)), False=Constant()) {defaultValue=true, parameterName=UseWPO}
 ]
}
