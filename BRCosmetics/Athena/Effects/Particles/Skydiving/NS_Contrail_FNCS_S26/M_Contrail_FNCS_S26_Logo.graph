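# Translucent unlit Niagara mesh-particle material for the FNCS S26 skydiving contrail logo.
# A panning color texture is added to particle color to feed emissive terms, including one scaled by logo alpha.
# Opacity-style math combines logo red and sharpened alpha with particle alpha and a radial falloff.
# A time-driven sine over a sphere mask builds a Z-axis wave, its strength blended by a dynamic parameter.
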Material /BRCosmetics/Athena/Effects/Particles/Skydiving/NS_Contrail_FNCS_S26/M_Contrail_FNCS_S26_Logo
graph {
 properties {
  bUsedWithNiagaraMeshParticles=true
  blendMode=BLEND_Translucent
  pixelDepthOffsetMode=PDOM_Legacy
  shadingModel=MSM_Unlit
  usedShadingModels=Unlit
 }
 nodes [
  n1 = Multiply(A=TextureCoordinate(), B=Constant() {r=0.1}) {constB=0.2}
  n2 = Multiply(A=Time(), B=Constant2Vector() {g=0.1, r=0.2})
  n3 = ParticleColor()
  n4 = Add(
   A=TextureSample(UVs=Add(A=@n1, B=@n2)) {texture=&/T_Contrail_FNCS_S26_Color2}.RGB
   B=@n3.RGB
  )
  MaterialFunctionCall(Emissive=@n4) {materialFunction=&MF_FX_EmissiveAdjustments}
  n5 = ScalarParameter() {defaultValue=8, parameterName=WPOStrenght}
  n6 = DynamicParameter() {paramNames [AmplitudeStrenght, Time, Opening, WPOStrenght]}
  n7 = Constant() {r=0.5}
  n8 = Multiply(A=SphereMask(A=TextureCoordinate(), B=@n7, Radius=@n7, Hardness=Constant())) {constB=1.5}
  n9 = Multiply(A=Time(), B=ScalarParameter() {defaultValue=2, parameterName=TimeMult}) {constB=2}
  Multiply(
   A=Multiply(A=Constant3Vector() {constant=<rgba 0 0 1 0>}, B=LinearInterpolate(A=@n5, Alpha=@n6.WPOStrenght) {constA=5, constB=0}) {constB=5}
   B=Sine(None=Add(A=@n8, B=@n9))
  )
  n10 = ScalarParameter() {defaultValue=1, parameterName=UFlip}
  n11 = ScalarParameter() {defaultValue=1, parameterName=VFlip}
  n12 = TextureSample(UVs=Multiply(A=TextureCoordinate(), B=AppendVector(A=@n10, B=@n11))) {samplerType=SAMPLERTYPE_LinearColor, texture=&/T_Contrail_FNCS_S26_Logo}
  Multiply(A=@n4, B=Multiply(A=@n12.A) {constB=2})
  n13 = Subtract(A=TextureCoordinate()) {constB=0.5}
  n14 = Power(Base=OneMinus(None=MaterialFunctionCall(Vector 2=@n13) {materialFunction=/Engine/Functions/Engine_MaterialFunctions02/Utility/VectorLength}.V2 Length)) {constExponent=6}
  n15 = Power(Base=@n12.A) {constExponent=4.5}
  Multiply(A=Saturate(None=Multiply(A=@n14) {constB=3}), B=Multiply(A=@n3.A, B=Multiply(A=@n12.R, B=@n15)))
 ]
}
