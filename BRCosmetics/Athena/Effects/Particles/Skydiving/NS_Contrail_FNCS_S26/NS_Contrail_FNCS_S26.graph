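# Niagara skydiving contrail cosmetic for FNCS S26 that dresses the player's hands and feet in glowing trail effects.
# Spawn groups index skeletal sockets, placing persistent arrow and tube meshes, glow lights, ribbons, sparks, stars and chunks, with hand chunks on GPU.
# Glide lean drives material parameters, dive state boosts ribbon velocity, and front-end modifiers swap motion values.
# LogoUnderPlayer is disabled, while "Empty" still renders a color-cycling mesh at socket index 4.
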NiagaraSystem /BRCosmetics/Athena/Effects/Particles/Skydiving/NS_Contrail_FNCS_S26/NS_Contrail_FNCS_S26
dependencies {
 usedDataInterfaces [
  &NiagaraDataInterfaceColorCurve
  &NiagaraDataInterfaceCurve
  &NiagaraDataInterfaceArrayColor
  &NiagaraDataInterfaceMeshRendererInfo
 ]
 usedDynamicInputs [&self:Abs_Z_Axis]
 usedModules [
  &self:SystemSpawnScript.SystemScriptSource.SystemScriptGraph.NiagaraNodeAssignment_3.SetVariables_#1
  &SystemState
  &self:SystemSpawnScript.SystemScriptSource.SystemScriptGraph.NiagaraNodeAssignment_1.SetVariables_#2
  &InterpolateOverTime
  &self:SystemSpawnScript.SystemScriptSource.SystemScriptGraph.NiagaraNodeAssignment_0.SetVariables_#3
  &self:SystemSpawnScript.SystemScriptSource.SystemScriptGraph.NiagaraNodeAssignment_2.SetVariables_#4
 ]
}
emitters {
 Arrows {
  emitterUpdateScript [
   &EmitterState(Life Cycle Mode=::NewEnumerator1, Loop Behavior=::NewEnumerator1, Loop Duration Mode=::NewEnumerator1, UseLoopDelay=false)
   &SpawnBurst_Instantaneous(Spawn Count=2, Spawn Probability=HlslExpression() {hlslExpression="Emitter.LoopCount == 0"}, Use Spawn Probability=true)
  ]
  particleSpawnScript [
   &FN_SetSpawnGroup(Number of Spawn Groups=2)
   &InitializeParticle(Mass Mode=::NewEnumerator1, Sprite Size Mode=::NewEnumerator4, UsePositionOffset=false, Write Material Random=true) {module=InitializeParticle001}
   &InitialMeshOrientation(Mesh Orientation Mode=::NewEnumerator2, Orientation Vector=<0 1 0>, Use Rotation Vector=false) {enabled=false}
  ]
  particleUpdateScript [
   &ParticleState(Kill Particles When Lifetime Has Elapsed=false)
   &DynamicMaterialParameters(
    Index 0 Param 1=&Lerp_Float(A=0.6, B=1, Alpha=$User.ParaGlideLeanAlpha)
    Index 0 Param 2=&Lerp_Float(A=5, B=10, Alpha=$User.ParaGlideLeanAlpha)
    Index 1 Param 1=HlslExpression() {hlslExpression=Engine.Owner.Velocity.x}
    Index 1 Param 2=HlslExpression() {hlslExpression=Engine.Owner.Velocity.y}
    Module.Param0WriteEnabled=true
    Module.Param0WriteEnabled1=true
    Module.Param1WriteEnabled=true
    Module.Param1WriteEnabled1=true
    Module.Param2WriteEnabled=true
    Module.Param2WriteEnabled1=true
    Module.Param3WriteEnabled=true
    Module.Param3WriteEnabled1=true
    Write Parameter Index 1=true
   )
   &SkeletalMeshLocation(
    Bone / Socket Index=$Particles.SpawnGroup
    Bone Sampling Mode=::NewEnumerator4
    Inherit Velocity=false
    Interpolate Mesh Sampling=false
    MaskSpawnGroup=false
    Sampled Position Offset=<0 0 5>
    Skeletal Mesh=$System.SkeletalMesh
   )
   /Niagara/Modules/Update/Orientation/V2/OrientMeshToVector(
    Override Incoming Orientation=1
    Override Quat=true
    Reference Axis Vector:Vector3f=<0 0 1>
    UseOptionalSideAxis=false
    UseOptionalUpAxis=false
   )
  ]
  rendererClasses [&NiagaraMeshRendererProperties]
  renderers [{rendererClass=&NiagaraMeshRendererProperties, rendererIndex=0}]
 }
 ArrowsFeet {
  emitterUpdateScript [
   &EmitterState(Life Cycle Mode=::NewEnumerator1, Loop Behavior=::NewEnumerator1, Loop Duration Mode=::NewEnumerator1, UseLoopDelay=false)
   &SpawnBurst_Instantaneous(Spawn Count=2, Spawn Probability=HlslExpression() {hlslExpression="Emitter.LoopCount == 0"}, Use Spawn Probability=true)
  ]
  particleSpawnScript [
   &FN_SetSpawnGroup(Number of Spawn Groups=2)
   &InitializeParticle(Mass Mode=::NewEnumerator1, Sprite Size Mode=::NewEnumerator4, UsePositionOffset=false, Write Material Random=true) {module=InitializeParticle001}
   &InitialMeshOrientation(Mesh Orientation Mode=::NewEnumerator2, Orientation Vector=<0 1 0>, Use Rotation Vector=false) {enabled=false}
  ]
  particleUpdateScript [
   &ParticleState(Kill Particles When Lifetime Has Elapsed=false)
   &DynamicMaterialParameters(
    Index 0 Param 1=&Lerp_Float(A=0.6, B=0.7, Alpha=$User.ParaGlideLeanAlpha)
    Index 0 Param 2=&Lerp_Float(A=5, B=15, Alpha=$User.ParaGlideLeanAlpha)
    Index 0 Param 3=$User.ParaGlideLeanAlpha
    Module.Param0WriteEnabled=true
    Module.Param1WriteEnabled=true
    Module.Param2WriteEnabled=true
    Module.Param3WriteEnabled=true
   )
   &SkeletalMeshLocation(
    Bone / Socket Index=&Add_Integer(A=$Particles.SpawnGroup, B=2)
    Bone Sampling Mode=::NewEnumerator4
    Inherit Velocity=false
    Interpolate Mesh Sampling=false
    MaskSpawnGroup=false
    Sampled Position Offset=<0 0 10>
    Skeletal Mesh=$System.SkeletalMesh
   )
   &OrientMeshToVector(
    Override Incoming Orientation=1
    Override Quat=true
    Reference Axis Vector:Vector3f=<0 0 1>
    UseOptionalSideAxis=false
    UseOptionalUpAxis=false
   )
  ]
  rendererClasses [&NiagaraMeshRendererProperties]
  renderers [{rendererClass=&NiagaraMeshRendererProperties, rendererIndex=0}]
 }
 ChunksFeet {
  emitterUpdateScript [
   &EmitterState(Life Cycle Mode=::NewEnumerator1, Loop Duration=0, UseLoopDelay=false)
   /Game/Effects/Niagara/Modules/Spawn/SpawnRateFromMultipleSpawnGroups_2(Global Spawnrate=8)
   &SpawnRate(SpawnRate=&Lerp_Float(A=10, B=25, Alpha=$User.ParaGlideLeanAlpha), Use Spawn Probability=false) {enabled=false}
  ]
  particleSpawnScript [
   &FN_SetSpawnGroup(Number of Spawn Groups=2)
   &InitializeParticle(
    AdjustAlpha=false
    AdjustHue=true
    AdjustSaturation=false
    AdjustValue=false
    Color=<rgba 2 0.5051 0 1>
    Color Mode=::NewEnumerator3
    Hue Shift Range:Vector2f=<-0.2 0.1>
    Lifetime Max=0.5
    Lifetime Min=0.25
    Lifetime Mode=::NewEnumerator1
    Mesh Renderer Array Visibility Mode=::NewEnumerator3
    Mesh Renderer Info=DataInterface() {propertyValues {MeshRenderer=&self:ChunksFeet_0.NiagaraMeshRendererProperties_0}}
    Mesh Scale Mode=::NewEnumerator1
    Mesh Uniform Scale Max=1
    Mesh Uniform Scale Min=0.35
    UsePositionOffset=false
    Write Material Random=true
   ) {module=InitializeParticle001}
   &SkeletalMeshLocation(
    Bone / Socket Index=&Add_Integer(A=$Particles.SpawnGroup, B=2)
    Bone Sampling Mode=::NewEnumerator4
    Inherit Velocity=false
    Interpolate Mesh Sampling=false
    MaskSpawnGroup=false
    Sampled Position Offset=<0 0 10>
    Skeletal Mesh=$System.SkeletalMesh
   )
   &ShapeLocation(MaskSpawnGroup=false, Override Seed=false, Sphere Radius=10, UseOwnerScale=true)
   &AddVelocity(Velocity=&FN_FrontEndVector3Modifier(
     Gameplay=&RandomRangeVector(Minimum=<50 50 125>, Maximum=<-50 -50 220>)
     Front End=&RandomRangeVector(Minimum=<0 0 85>, Maximum=<1 1 180>)
    ))
   &InitialMeshOrientation(Use Rotation Vector=false)
  ]
  particleUpdateScript [
   &ParticleState()
   &ScaleMeshSize(Scale Factor=&VectorFromFloat(Value=&FloatFromCurve(FloatCurve=DataInterface() {propertyValues {Curve {keys [
          {arriveTangent=9.0725, interpMode=RCIM_Cubic, leaveTangent=9.0725, tangentMode=RCTM_User, time=0, value=0}
          {arriveTangent=0.5485, interpMode=RCIM_Cubic, leaveTangent=0.5485, tangentMode=RCTM_User, time=0.1408, value=0.9847}
          {arriveTangent=-2.2389, interpMode=RCIM_Cubic, leaveTangent=-2.2389, tangentMode=RCTM_User, time=1, value=0}
         ]}}})))
   &Drag(Drag=3, Use Linear Drag=true, Use Rotational Drag=true)
   &SolveForcesAndVelocity(Clamp Velocity=false, Limit Acceleration=false)
  ]
  rendererClasses [&NiagaraMeshRendererProperties]
  renderers [{rendererClass=&NiagaraMeshRendererProperties, rendererIndex=0}]
 }
 ChunksHand {
  simTarget=GPUComputeSim
  emitterUpdateScript [
   &EmitterState(Life Cycle Mode=::NewEnumerator1, Loop Duration=1, UseLoopDelay=false)
   &SpawnRateFromMultipleSpawnGroups_2(Global Spawnrate=50)
   &SpawnRate(SpawnRate=&Lerp_Float(A=10, B=25, Alpha=$User.ParaGlideLeanAlpha), Use Spawn Probability=false) {enabled=false}
  ]
  particleSpawnScript [
   &FN_SetSpawnGroup(Number of Spawn Groups=2)
   &InitializeParticle(
    AdjustAlpha=false
    AdjustHue=true
    AdjustSaturation=false
    AdjustValue=false
    Color=<rgba 2 0.5051 0 1>
    Color Mode=::NewEnumerator3
    Hue Shift Range:Vector2f=<-0.2 0.1>
    Lifetime Max=0.5
    Lifetime Min=0.25
    Lifetime Mode=::NewEnumerator1
    Mesh Renderer Array Visibility Mode=::NewEnumerator3
    Mesh Renderer Info=DataInterface() {propertyValues {MeshRenderer=&self:ChunksHand_3.NiagaraMeshRendererProperties_0}}
    Mesh Scale Mode=::NewEnumerator1
    Mesh Uniform Scale Max=1
    Mesh Uniform Scale Min=0.35
    UsePositionOffset=false
    Write Material Random=true
   ) {module=InitializeParticle001}
   &SkeletalMeshLocation(
    Bone / Socket Index=$Particles.SpawnGroup
    Bone Sampling Mode=::NewEnumerator4
    Inherit Velocity=false
    Interpolate Mesh Sampling=false
    MaskSpawnGroup=false
    Sampled Position Offset=<0 0 40>
    Skeletal Mesh=$System.SkeletalMesh
   )
   &ShapeLocation(
    Cylinder Radius=10
    MaskSpawnGroup=false
    Override Seed=false
    Shape Primitive=::NewEnumerator1
    Surface Only=true
    Surface Only Band Thickness=1
    UseOwnerScale=true
   )
   &AddVelocity(Velocity=&FN_FrontEndVector3Modifier(
     Gameplay=&RandomRangeVector(Minimum=<100 100 220>, Maximum=<-100 -100 350>)
     Front End=&RandomRangeVector(Minimum=<0 0 85>, Maximum=<1 1 180>)
    ))
   &InitialMeshOrientation(Use Rotation Vector=false)
  ]
  particleUpdateScript [
   &ParticleState()
   &ScaleMeshSize(Scale Factor=&VectorFromFloat(Value=&FloatFromCurve(FloatCurve=DataInterface() {propertyValues {Curve {keys [
          {arriveTangent=9.0725, interpMode=RCIM_Cubic, leaveTangent=9.0725, tangentMode=RCTM_User, time=0, value=0}
          {arriveTangent=0.5485, interpMode=RCIM_Cubic, leaveTangent=0.5485, tangentMode=RCTM_User, time=0.1408, value=0.9847}
          {arriveTangent=-3.2166, interpMode=RCIM_Cubic, leaveTangent=-3.2166, tangentMode=RCTM_User, time=1, value=0}
         ]}}})))
   &Drag(Drag=3, Use Linear Drag=true, Use Rotational Drag=true)
   &SolveForcesAndVelocity(Clamp Velocity=false, Limit Acceleration=false)
  ]
  rendererClasses [&NiagaraMeshRendererProperties]
  renderers [{rendererClass=&NiagaraMeshRendererProperties, rendererIndex=0}]
 }
 CrackedTube {
  emitterUpdateScript [
   &EmitterState(Life Cycle Mode=::NewEnumerator1, Loop Behavior=::NewEnumerator1, Loop Delay=0.4, Loop Duration Mode=::NewEnumerator1, UseLoopDelay=true)
   &SpawnBurst_Instantaneous(Spawn Count=2, Spawn Probability=HlslExpression() {hlslExpression="Emitter.LoopCount == 0"}, Use Spawn Probability=true)
  ]
  particleSpawnScript [
   &FN_SetSpawnGroup()
   &InitializeParticle(Mass Mode=::NewEnumerator1, Sprite Size Mode=::NewEnumerator4, UsePositionOffset=false, Write Material Random=true) {module=InitializeParticle001}
   &InitialMeshOrientation(
    Mesh Orientation Mode=::NewEnumerator2
    Orientation Coordinate Space=::Local
    Rotation=&RandomRangeVector(Minimum=<0 0 -1>, Maximum=<0 0 1>)
    Rotation Coordinate Space=::NewEnumerator0
    Use Rotation Vector=true
   )
  ]
  particleUpdateScript [
   &ParticleState(Kill Particles When Lifetime Has Elapsed=false)
   &SkeletalMeshLocation(
    Bone / Socket Index=$Particles.SpawnGroup
    Bone Sampling Mode=::NewEnumerator4
    Inherit Velocity=false
    Interpolate Mesh Sampling=false
    MaskSpawnGroup=false
    Sampled Position Offset=<0 0 6>
    Skeletal Mesh=$System.SkeletalMesh
   )
   &DynamicMaterialParameters(
    Index 0 Param 1=&Lerp_Float(A=25, B=40, Alpha=$User.ParaGlideLeanAlpha)
    Index 0 Param 2=1
    Index 0 Param 4=$User.ParaGlideLeanAlpha
    Index 1 Param 1=HlslExpression() {hlslExpression=Engine.Owner.Velocity.x}
    Index 1 Param 2=HlslExpression() {hlslExpression=Engine.Owner.Velocity.y}
    Module.Param0WriteEnabled=true
    Module.Param0WriteEnabled1=true
    Module.Param1WriteEnabled=true
    Module.Param1WriteEnabled1=true
    Module.Param2WriteEnabled=true
    Module.Param2WriteEnabled1=true
    Module.Param3WriteEnabled=true
    Module.Param3WriteEnabled1=true
    Write Parameter Index 1=true
   ) {module=DynamicMaterialParameters001}
  ]
  rendererClasses [&NiagaraMeshRendererProperties]
  renderers [
   {rendererClass=&NiagaraMeshRendererProperties, rendererIndex=0}
   {rendererClass=&NiagaraMeshRendererProperties, rendererIndex=1}
  ]
 }
 CrackedTubeInitiate {
  emitterUpdateScript [
   &EmitterState(Life Cycle Mode=::NewEnumerator1, Loop Behavior=::NewEnumerator1, Loop Duration=1, UseLoopDelay=false)
   &SpawnBurst_Instantaneous(Spawn Count=2, Spawn Probability=HlslExpression() {hlslExpression="Emitter.LoopCount == 0"}, Use Spawn Probability=true)
  ]
  particleSpawnScript [
   &FN_SetSpawnGroup(Number of Spawn Groups=2)
   &InitializeParticle(Lifetime=0.45, Mass Mode=::NewEnumerator1, Sprite Size Mode=::NewEnumerator4, UsePositionOffset=false, Write Material Random=true) {module=InitializeParticle001}
  ]
  particleUpdateScript [
   &ParticleState()
   &SkeletalMeshLocation(
    Bone / Socket Index=$Particles.SpawnGroup
    Bone Sampling Mode=::NewEnumerator4
    Inherit Velocity=false
    Interpolate Mesh Sampling=false
    MaskSpawnGroup=false
    Sampled Position Offset=<0 0 6>
    Skeletal Mesh=$System.SkeletalMesh
   )
   &DynamicMaterialParameters(
    Index 0 Param 1=&FloatFromCurve(FloatCurve=DataInterface() {propertyValues {Curve {keys [{interpMode=RCIM_Cubic, leaveTangent=-115.7559, tangentMode=RCTM_Break, time=0, value=50}, <at 1 1>]}}})
    Index 0 Param 2=&FloatFromCurve(FloatCurve=DataInterface() {propertyValues {Curve {keys [<at 0 0.1>, {interpMode=RCIM_Cubic, leaveTangent=2.0518, tangentMode=RCTM_Break, time=0.3, value=0.5}, <at 1 1>]}}})
    Index 0 Param 3=1
    Index 0 Param 4=&FN_FrontEndFloatModifier(Gameplay=0, Front End=0.6)
    Module.Param0WriteEnabled=true
    Module.Param1WriteEnabled=true
    Module.Param2WriteEnabled=true
    Module.Param3WriteEnabled=true
   ) {module=DynamicMaterialParameters001}
   &OrientMeshToVector(
    Override Incoming Orientation=1
    Override Quat=true
    Reference Axis Vector:Vector3f=<0 0 1>
    UseOptionalSideAxis=false
    UseOptionalUpAxis=false
   )
  ]
  rendererClasses [&NiagaraMeshRendererProperties]
  renderers [{rendererClass=&NiagaraMeshRendererProperties, rendererIndex=0}]
 }
 Decal {
  emitterUpdateScript [
   &EmitterState(Life Cycle Mode=::NewEnumerator1, Loop Duration=1, UseLoopDelay=false)
   &SpawnRateFromMultipleSpawnGroups_2(Global Spawnrate=4)
   &SpawnRate(SpawnRate=14, Use Spawn Probability=false) {enabled=false}
  ]
  particleSpawnScript [
   &FN_SetSpawnGroup()
   /Niagara/Modules/Spawn/Initialization/InitializeParticle(
    Color=<rgba 0.9062 0.4123 0.1067 0.6>
    Lifetime=0.6
    Module.Write Color=true
    Module.Write Lifetime=true
    Module.Write Mass=true
    Module.Write Material Random=true
    Module.Write Position=false
    Module.Write Scale=false
    Module.Write SpriteRotation=false
    Module.Write SpriteSize=true
    Position=&SimulationPosition()
    Sprite Size=<40 40>
   ) {enabled=false}
   /Niagara/Modules/Spawn/Initialization/V2/InitializeParticle(
    Color=<rgba 0.9062 0.4123 0.1067 0.6>
    Lifetime=0.6
    Mass Mode=::NewEnumerator1
    Sprite Size=<40 40>
    Sprite Size Mode=::NewEnumerator4
    UsePositionOffset=false
    Write Material Random=true
   ) {module=InitializeParticle001}
   &CameraOffset(Camera Offset Amount=9)
  ]
  particleUpdateScript [
   &ParticleState()
   &SkeletalMeshLocation(
    Bone / Socket Index=$Particles.SpawnGroup
    Bone Sampling Mode=::NewEnumerator4
    Inherit Velocity=false
    Interpolate Mesh Sampling=false
    MaskSpawnGroup=false
    Skeletal Mesh=$System.SkeletalMesh
   )
   &ScaleColor(
    Scale Alpha=&FloatFromCurve(
     FloatCurve=DataInterface() {propertyValues {Curve {keys [<at 0.0001 0.3382>,<at 0.1393 0.6251>,<at 0.3202 0.5233>,<at 0.4613 0.8934>,<at 0.6241 0.3845>,<at 0.7724 0.6066>,<at 1.0006 0.3567>]}}}
     Scale Curve=HlslExpression() {hlslExpression="Particles.SpawnGroup < 2 ? 1 : 0.5"}
    )
    ScaleA=true
    ScaleRGB=false
   )
  ]
  rendererClasses [&NiagaraSpriteRendererProperties]
  renderers [{rendererClass=&NiagaraSpriteRendererProperties}]
 }
 Empty {
  emitterUpdateScript [
   &EmitterState(Life Cycle Mode=::NewEnumerator1, Loop Behavior=::NewEnumerator1, Loop Duration Mode=::NewEnumerator1, UseLoopDelay=false)
   &SpawnBurst_Instantaneous(Use Spawn Probability=false, UseLoopCountLimit=false)
  ]
  particleSpawnScript [
   &FN_SetSpawnGroup(Number of Spawn Groups=1)
   &InitializeParticle(
    Color=<rgba 1 1 1 0.65>
    Lifetime=5
    Mesh Scale=<2.3 2.3 1.1>
    Mesh Scale Mode=::NewEnumerator4
    UsePositionOffset=false
    Write Material Random=true
   )
   &InitialMeshOrientation(
    Mesh Orientation Mode=::NewEnumerator2
    Orientation Vector=<0 0 1>
    Rotation=<0.038 0.03 0.24>
    Rotation Coordinate Space=::NewEnumerator2
    Use Rotation Vector=true
   )
  ]
  particleUpdateScript [
   &ParticleState(Kill Particles When Lifetime Has Elapsed=false)
   &UpdateMeshOrientation(
    Rotation Rate=&MakeCustomFloatFromBool(Bool=&SetBoolByFloatComparison(A=$User.ParaGlideLeanAlpha, B=0.5), True Float=-8, False Float=0)
    Rotation Vector=<1 0.2 0.06>
   ) {enabled=false}
   &SkeletalMeshLocation(
    Bone / Socket Index=4
    Bone Sampling Mode=::NewEnumerator4
    MaskSpawnGroup=false
    Sampled Position Offset=<35 30 0>
    Skeletal Mesh=$System.SkeletalMesh
   )
   &ScaleColor(
    Curve Index=$System.LoopedAge
    Linear Color Curve=DataInterface() {propertyValues {
      AlphaCurve {keys [<at 0 1>,<at 0.3628 0.2>,<at 1 1>]}
      BlueCurve {keys [<at 0 0.1559>,<at 0.1118 0.0052>,<at 0.3326 3>,<at 0.521 1.305>,<at 0.7478 2>,<at 0.9365 0.0865>]}
      GreenCurve {keys [<at 0 0.4735>,<at 0.1118 0.2961>,<at 0.3326 0.8595>,<at 0.521 2>,<at 0.7478 0.3332>,<at 0.9365 0.1962>]}
      RedCurve {keys [<at 0 1>,<at 0.1118 1>,<at 0.3326 0.2585>,<at 0.521 0.1891>,<at 0.7478 1.1865>,<at 0.9365 1>]}
     }}
    Scale Mode=::NewEnumerator2
   )
   &DynamicMaterialParameters(
    Index 0 Param 2=2
    Index 0 Param 3=1
    Index 0 Param 4=0.65
    Module.Param0WriteEnabled=true
    Module.Param0WriteEnabled1=true
    Module.Param1WriteEnabled=true
    Module.Param1WriteEnabled1=true
    Module.Param2WriteEnabled=true
    Module.Param2WriteEnabled1=true
    Module.Param3WriteEnabled=true
    Module.Param3WriteEnabled1=true
    Write Parameter Index 1=true
   )
  ]
  rendererClasses [&NiagaraMeshRendererProperties]
  renderers [
   {rendererClass=&NiagaraMeshRendererProperties, rendererIndex=0}
   {rendererClass=&NiagaraMeshRendererProperties, rendererIndex=1}
  ]
 }
 HandGlow {
  emitterUpdateScript [
   &EmitterState(Life Cycle Mode=::NewEnumerator1, Loop Behavior=::NewEnumerator1, Loop Duration Mode=::NewEnumerator1, UseLoopDelay=false)
   &SpawnBurst_Instantaneous(Spawn Count=4, Spawn Probability=HlslExpression() {hlslExpression="Emitter.LoopCount == 0"}, Use Spawn Probability=true)
  ]
  particleSpawnScript [
   &FN_SetSpawnGroup()
   &InitializeParticle(
    Color=<rgba 2 2 2 0.25>
    Lifetime=2
    Mass Mode=::NewEnumerator1
    Sprite Size=<40 40>
    Sprite Size Mode=::NewEnumerator4
    UsePositionOffset=false
    Write Material Random=true
   ) {module=InitializeParticle001}
  ]
  particleUpdateScript [
   &ParticleState(Kill Particles When Lifetime Has Elapsed=false)
   &SkeletalMeshLocation(
    Bone / Socket Index=$Particles.SpawnGroup
    Bone Sampling Mode=::NewEnumerator4
    Inherit Velocity=false
    Interpolate Mesh Sampling=false
    MaskSpawnGroup=false
    Sampled Position Offset=<0 0 4>
    Skeletal Mesh=$System.SkeletalMesh
   )
   &CameraOffset(Camera Offset Amount=15)
   &ScaleColor(
    Curve Index=$System.LoopedAge
    Linear Color Curve=DataInterface() {propertyValues {
      AlphaCurve {keys [<at 0 1>,<at 0.4671 0.6>,<at 1 1>]}
      BlueCurve {keys [<at 0 0>,<at 0.0548 0.0052>,<at 0.1868 1>,<at 0.3164 1>,<at 0.5273 0.7379>,<at 0.8205 0.849>,<at 1 0>]}
      GreenCurve {keys [<at 0 0.5885>,<at 0.0548 0.2961>,<at 0.1868 0.3821>,<at 0.3164 0.298>,<at 0.5273 1>,<at 0.8205 0.2399>,<at 1 0.5684>]}
      RedCurve {keys [<at 0 1>,<at 0.0548 1>,<at 0.1868 0.0916>,<at 0.3164 0>,<at 0.5273 0.0242>,<at 0.8205 0.2823>,<at 1 1>]}
     }}
    Scale Mode=::NewEnumerator2
   )
   &FN_Light_v2(Light Exponent=2, Light Radius=100, LightEnabled=true, Scale Color=2)
  ]
  rendererClasses [&NiagaraSpriteRendererProperties, &NiagaraLightRendererProperties]
  renderers [
   {rendererClass=&NiagaraSpriteRendererProperties}
   {rendererClass=&NiagaraLightRendererProperties, rendererIndex=1}
  ]
 }
 LogoUnderPlayer {
  bEnabled=false
  emitterUpdateScript [
   &EmitterState(Life Cycle Mode=::NewEnumerator1, Loop Behavior=::NewEnumerator1, Loop Duration Mode=::NewEnumerator1, UseLoopDelay=false)
   &SpawnBurst_Instantaneous(Spawn Count=1, Spawn Probability=HlslExpression() {hlslExpression="Emitter.LoopCount == 0"}, Use Spawn Probability=true)
  ]
  particleSpawnScript [
   &FN_SetSpawnGroup(Number of Spawn Groups=1)
   &InitializeParticle(Color=<rgba 1 1 1 0.65>, Mesh Scale=<1.8 1.8 1.2>, Mesh Scale Mode=::NewEnumerator4, UsePositionOffset=false, Write Material Random=true) {module=InitializeParticle001}
   &InitialMeshOrientation(
    Mesh Orientation Mode=::NewEnumerator2
    Orientation Coordinate Space=::Local
    Orientation Vector=<0 0 1>
    Rotation=<-0.45 0 0>
    Rotation Coordinate Space=::NewEnumerator2
    Use Rotation Vector=true
   )
  ]
  particleUpdateScript [
   &ParticleState(Kill Particles When Lifetime Has Elapsed=false)
   &SystemLocation(MaskSpawnGroup=false, Offset=<0 5 50>)
   &SkeletalMeshLocation(
    Bone / Socket Index=4
    Bone Sampling Mode=::NewEnumerator4
    MaskSpawnGroup=false
    Orientation Sampling=::NewEnumerator1
    Sampled Position Offset=<0 0 -60>
    Sampled Position Offset Coordinate Space:ENiagaraSkelMeshTransforms=::NewEnumerator0
    Skeletal Mesh=$User.SkeletalMesh
   ) {enabled=false}
   /SetVariables(Particles.MeshOrientation=/Niagara/DynamicInputs/Vectors/OrientMeshToVector(
     Look At Direction=&FN_FrontEndVector3Modifier(
      Gameplay=&Multiply_VectorByFloat(
       Vector=&NormalizeVector(
        Vector To Normalize=/Niagara/DynamicInputs/Vectors/NegateVector(Vector:Vector3f=&TransformVector(
          Vector=&Multiply_Vector(A=$Engine.Owner.Velocity, B=<0.55 0.55 1>)
          Source Space=::World
          Destination Space=::Local
         ))
        Safe Normalize=true
        Safe Fallback Threshold=0
        Safe Fallback Vector=<0 0 0.001>
       )
       Float=-1
      )
      Front End=<0 0.25 -1>
     )
     Reference Axis Vector=<1 0 0>
    )) {isSetParameters=true}
   /SetVariables(Particles.MeshOrientation=&OrientMeshToVector(
     Look At Direction=&FN_FrontEndVector3Modifier(
      Gameplay=&Lerp_Vector(
       A=&RotateVectorByQuaternion(Vector To Rotate=$Engine.Owner.Velocity, Quaternion=/Niagara/DynamicInputs/Rotation/InvertQuaternion(Quat:Quat4f=$Engine.Owner.Rotation))
       B=<0 0 1>
       Alpha=$System.Dive
      )
      Front End=&Lerp_Vector(A=<0 0 -1>, B=<0 2 -1>, Alpha=$System.Dive)
     )
     Reference Axis Vector=<1 0 0>
    )) {enabled=false, isSetParameters=true}
   &ScaleColor(
    Curve Index=$System.LoopedAge
    Linear Color Curve=DataInterface() {propertyValues {
      AlphaCurve {keys [<at 0 1>,<at 0.3628 0.2>,<at 1 1>]}
      BlueCurve {keys [<at 0 0.1559>,<at 0.1713 0.0052>,<at 0.3326 0.429>,<at 0.4777 1.305>,<at 0.7478 0>,<at 1 0.1559>]}
      GreenCurve {keys [<at 0 0.4735>,<at 0.1713 0.2961>,<at 0.3326 0.3096>,<at 0.4777 2>,<at 0.7478 0.852>,<at 1 0.4735>]}
      RedCurve {keys [<at 0 1>,<at 0.1713 1>,<at 0.3326 3>,<at 0.4777 0.1891>,<at 0.7478 2>,<at 1 1>]}
     }}
    Scale Mode=::NewEnumerator2
   )
   &DynamicMaterialParameters(
    Index 0 Param 2=2
    Index 0 Param 3=1
    Index 0 Param 4=0.65
    Module.Param0WriteEnabled=true
    Module.Param0WriteEnabled1=true
    Module.Param1WriteEnabled=true
    Module.Param1WriteEnabled1=true
    Module.Param2WriteEnabled=true
    Module.Param2WriteEnabled1=true
    Module.Param3WriteEnabled=true
    Module.Param3WriteEnabled1=true
    Write Parameter Index 1=true
   )
  ]
  rendererClasses [&NiagaraMeshRendererProperties]
  renderers [
   {rendererClass=&NiagaraMeshRendererProperties, rendererIndex=0}
   {rendererClass=&NiagaraMeshRendererProperties, rendererIndex=1}
  ]
 }
 Ribbons {
  emitterUpdateScript [
   &EmitterState(Life Cycle Mode=::NewEnumerator1, Loop Duration=1, UseLoopDelay=false)
   /Game/Effects/Niagara/Modules/Skydiving/FN_DivingAttributes/FN_EmitterDivingAttributes_v1(Bone Index=1, SkeletalMesh:NiagaraDataInterfaceSkeletalMesh=$User.SkeletalMesh)
   &FN_SpawnRibbon_v2(Number of Ribbons=4, Spawn Mode=::NewEnumerator0, SpawnRate=64)
  ]
  particleSpawnScript [
   &FN_SetSpawnGroup()
   &FN_InitializeRibbon_v3(
    Color:LinearColor=&SelectLinearColorFromArray(
     Array Sampling Mode=::NewEnumerator1
     Direct Array Index=&SpawnGroup()
     Color Selection Array=DataInterface() {propertyValues {ColorData [<rgba 0.5188 0.9702 3 0.6>, <rgba 0.5188 0.9702 3 0.6>, <rgba 3 1.2562 0.2784 0.4>, <rgba 3 1.2562 0.2784 0.4>]}}
     Random Seed=?
    )
    Custom Offset=true
    Enable Skeletal Position=true
    Lifetime=HlslExpression() {hlslExpression="Particles.SpawnGroup < 2 ? 0.26 : 0.11"}
    Local Vector Offset:Vector3f=<0 1 0>
    Ribbon Width=HlslExpression() {hlslExpression="Particles.SpawnGroup < 2 ? 16 : 8"}
   )
   &AddVelocity(
    Scale Added Velocity=&Lerp_Vector(A=<1 1 1>, B=<4 4 1>, Alpha=$System.Dive)
    Velocity=&FN_FrontEndVector3Modifier(Gameplay=<0 0 800>, Front End=<0 0 800>)
   )
   &SkeletalMeshLocation(
    Apply Sampled Orientation to Mesh Particles=true
    Bone / Socket Index=$Particles.SpawnGroup
    Bone Sampling Mode=::NewEnumerator4
    Inherit Velocity=false
    Interpolate Mesh Sampling=false
    MaskSpawnGroup=false
    Sampled Position Offset Coordinate Space:ENiagaraMeshTransforms=::NewEnumerator3
    Skeletal Mesh=$System.SkeletalMesh
   )
   &DynamicMaterialParameters(
    Index 0 Param 1=&MaskFloatBySpawnGroup(Masked Float=1, Passthrough Float=0.8)
    Index 0 Param 2=&MaskFloatBySpawnGroup(Masked Float=0.4, Passthrough Float=3)
    Index 0 Param 3=1
    Index 1 Param 1=&Lerp_Float(A=5, B=10, Alpha=$System.Dive)
    Module.Param0WriteEnabled=true
    Module.Param0WriteEnabled1=true
    Module.Param1WriteEnabled=true
    Module.Param1WriteEnabled1=false
    Module.Param2WriteEnabled=true
    Module.Param2WriteEnabled1=false
    Module.Param3WriteEnabled=false
    Module.Param3WriteEnabled1=false
    Write Parameter Index 1=true
   )
  ]
  particleUpdateScript [
   &InheritSourceMovement(Applied Position Falloff Distance=100, Applied Position Scale=<0.9 0.9 1>, ApplyForce=false, ApplyPosition=true, ApplyPositionFalloff=false)
   &SolveForcesAndVelocity(Clamp Velocity=false, Limit Acceleration=false)
   &FN_UpdateRibbon(Detach Locked Particle=true, Enable Skeletal Position=true)
   &ScaleRibbonWidth(Ribbon Width Scale=&FloatFromCurve(FloatCurve=DataInterface() {propertyValues {Curve {keys [<at 0 1>,<at 1 4>]}}}))
  ]
  rendererClasses [&NiagaraRibbonRendererProperties]
  renderers [{rendererClass=&NiagaraRibbonRendererProperties}]
 }
 Sparks {
  emitterUpdateScript [
   &EmitterState(Life Cycle Mode=::NewEnumerator1, Loop Duration=0, UseLoopDelay=false)
   &SpawnRateFromMultipleSpawnGroups_2(Global Spawnrate=20)
   &SpawnRate(SpawnRate=&Lerp_Float(A=10, B=25, Alpha=$User.ParaGlideLeanAlpha), Use Spawn Probability=false) {enabled=false}
  ]
  particleSpawnScript [
   &FN_SetSpawnGroup()
   &InitializeParticle(
    Color Maximum=<rgba 0.0951 0.5889 3 1>
    Color Minimum=<rgba 2 0.4218 0 1>
    Color Mode=::NewEnumerator2
    Lifetime Max=0.4
    Lifetime Min=0.15
    Lifetime Mode=::NewEnumerator1
    Sprite Size Mode=::NewEnumerator1
    Uniform Sprite Size Max=2.5
    Uniform Sprite Size Min=1
    UsePositionOffset=false
    Write Material Random=true
   ) {module=InitializeParticle001}
   &SkeletalMeshLocation(
    Bone / Socket Index=$Particles.SpawnGroup
    Bone Sampling Mode=::NewEnumerator4
    Inherit Velocity=false
    Interpolate Mesh Sampling=false
    MaskSpawnGroup=false
    Sampled Position Offset=<0 0 10>
    Skeletal Mesh=$System.SkeletalMesh
   )
   &ShapeLocation(MaskSpawnGroup=false, Override Seed=false, Sphere Radius=15, UseOwnerScale=true)
   &AddVelocity(Velocity=&FN_FrontEndVector3Modifier(
     Gameplay=&RandomRangeVector(Minimum=<0 0 200>, Maximum=<1 1 450>)
     Front End=&RandomRangeVector(Minimum=<0 0 85>, Maximum=<1 1 180>)
    ))
  ]
  particleUpdateScript [
   &ParticleState()
   &InheritVelocity(Inherited Velocity Amount Scale=<-0.03 -0.03 -0.03>, Inherited Velocity Speed Limit=1000, Speed Limit=true, UseSpeedThreshold=false)
   &Drag(Drag=2, Use Linear Drag=true, Use Rotational Drag=true)
   &SolveForcesAndVelocity(Clamp Velocity=false, Limit Acceleration=false)
   &ScaleSpriteSizeBySpeed(Max Scale Factor=<1 6>, Min Scale Factor=<0.5 4>)
  ]
  rendererClasses [&NiagaraSpriteRendererProperties]
  renderers [{rendererClass=&NiagaraSpriteRendererProperties}]
 }
 Stars {
  emitterUpdateScript [
   &EmitterState()
   &SpawnRateFromMultipleSpawnGroups_2(Global Spawnrate=25)
  ]
  particleSpawnScript [
   &FN_SetSpawnGroup(Number of Spawn Groups=2)
   &InitializeParticle(
    AdjustAlpha=false
    AdjustHue=true
    AdjustSaturation=false
    AdjustValue=false
    Color=<rgba 0.6898 1.8436 3 1>
    Color Mode=::NewEnumerator3
    Lifetime Max=0.3
    Lifetime Min=0.1
    Lifetime Mode=::NewEnumerator1
    Sprite Size Mode=::NewEnumerator1
    Uniform Sprite Size Max=20
    Uniform Sprite Size Min=12
    UsePositionOffset=false
    Write Material Random=true
   ) {module=InitializeParticle001}
   &SkeletalMeshLocation(
    Bone / Socket Index=$Particles.SpawnGroup
    Bone Sampling Mode=::NewEnumerator4
    Inherit Velocity=false
    Interpolate Mesh Sampling=false
    MaskSpawnGroup=false
    Skeletal Mesh=$System.SkeletalMesh
   )
   &ShapeLocation(
    Cylinder Height=150
    Cylinder Height Midpoint=0
    Cylinder Radius=20
    MaskSpawnGroup=false
    Override Seed=false
    Shape Primitive=::NewEnumerator1
    Surface Only=true
    Surface Only Band Thickness=1
    Use Endcaps In Surface Only Mode=false
    UseOwnerScale=true
   )
   &AddVelocity(
    Velocity=&RandomRangeVector(Minimum=<30 30 30>, Maximum=<-30 -30 -30>)
    Velocity Speed Scale=&Lerp_Float(A=1, B=1.5, Alpha=$User.ParaGlideLeanAlpha)
   )
  ]
  particleUpdateScript [
   &ParticleState()
   &ScaleSpriteSize(Uniform Curve Sprite Scale=DataInterface() {propertyValues {Curve {keys [{tangentMode=RCTM_User, time=0, value=0}, <at 0.5 1>, <at 0.7 1>, {tangentMode=RCTM_User, time=1, value=0.15}]}}})
   &CameraOffset(Camera Offset Amount=10)
   &SolveForcesAndVelocity(Clamp Velocity=false, Limit Acceleration=false)
  ]
  rendererClasses [&NiagaraSpriteRendererProperties]
  renderers [{rendererClass=&NiagaraSpriteRendererProperties}]
 }
}
summary {
 systemName=NS_Contrail_FNCS_S26
 emitters [
  {emitterName=Arrows, rendererClasses [&NiagaraMeshRendererProperties]}
  {emitterName=HandGlow, rendererClasses [&NiagaraSpriteRendererProperties, &NiagaraLightRendererProperties]}
  {emitterName=Sparks, rendererClasses [&NiagaraSpriteRendererProperties]}
  {emitterName=Ribbons, rendererClasses [&NiagaraRibbonRendererProperties]}
  {emitterName=CrackedTube, rendererClasses [&NiagaraMeshRendererProperties]}
  {emitterName=ArrowsFeet, rendererClasses [&NiagaraMeshRendererProperties]}
  {emitterName=CrackedTubeInitiate, rendererClasses [&NiagaraMeshRendererProperties]}
  {emitterName=ChunksHand, simTarget=GPUComputeSim, rendererClasses [&NiagaraMeshRendererProperties]}
  {emitterName=ChunksFeet, rendererClasses [&NiagaraMeshRendererProperties]}
  {emitterName=Decal, rendererClasses [&NiagaraSpriteRendererProperties]}
  {emitterName=Stars, rendererClasses [&NiagaraSpriteRendererProperties]}
  {bEnabled=false, emitterName=LogoUnderPlayer, rendererClasses [&NiagaraMeshRendererProperties]}
  {emitterName=Empty, rendererClasses [&NiagaraMeshRendererProperties]}
 ]
 userVariables [
  User.bIsFrontEnd {type=NiagaraBool}
  User.bIsFrontEndPreview {type=NiagaraBool}
  User.ParaGlideLeanAlpha {type=NiagaraFloat, defaultValue=0, description=""}
  User.RibbonIDOffset {type=NiagaraInt32}
  User.SkeletalMesh {
   type=NiagaraDataInterfaceSkeletalMesh
   defaultValue {
    struct=NiagaraExt_VariableValue_DataInterface
    value {dataInterface=&self:NiagaraDataInterfaceSkeletalMesh_1}
   }
  }
 ]
}
